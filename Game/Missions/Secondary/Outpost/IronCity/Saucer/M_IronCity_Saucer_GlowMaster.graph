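# Master glow material for the Iron City saucer, producing neon pulse lines and a swirling iris.
# A V-coordinate split selects scrolling sine pulses or a solid band plus rotating spiral, distorted and tinted by panning noise on UV1.
# SaucerState stages the effect: iris and band fade in first, lines next, then pulse boost and layered-sine flicker.
# The result feeds DeriveHDRfromLDR for HDR intensity, alongside glass surface parameters.
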Material /Game/Missions/Secondary/Outpost/IronCity/Saucer/M_IronCity_Saucer_GlowMaster
graph {
 properties {pixelDepthOffsetMode=PDOM_Legacy, usedShadingModels=Default Lit}
 nodes [
  n1 = TextureCoordinate() {uTiling=10, vTiling=4}
  n2 = Frac(None=Multiply(A=Time() {period=1}, B=ScalarParameter() {defaultValue=0.5, group=Neon, parameterName=Pulse_Speed}))
  n3 = Sine(None=Add(A=ComponentMask(None=@n1) {r=true}, B=LinearInterpolate(Alpha=@n2) {constA=1, constB=-1}) {constB=0}) {period=2}
  n4 = Power(
   Base=Multiply(A=Add(A=@n3)) {constB=0.5}
   Exp=ScalarParameter() {defaultValue=2.2, group=Neon, parameterName="Pulse_SizeConrol (Contrast)", sortPriority=3}
  ) {constExponent=6}
  n5 = Multiply(A=ScalarParameter() {defaultValue=1, parameterName=SaucerState, sliderMax=1}) {constB=3}
  n6 = Clamp(None=Subtract(A=@n5) {constB=2})
  n7 = Multiply(A=@n6) {constB=0.35}
  n8 = ComponentMask(None=TextureCoordinate()) {g=true}
  n9 = Constant() {r=0.25}
  n10 = ScalarParameter() {defaultValue=0.25, parameterName=LineMask-Radius}
  n11 = ScalarParameter() {defaultValue=0.6, parameterName=LineMask-Hardness}
  n12 = ScalarParameter() {defaultValue=1, parameterName=LineMask-Exponent}
  n13 = Multiply(
   A=Multiply(A=Add(A=@n4, B=@n7), B=VectorParameter() {parameterName=Color-PulseGlow, defaultValue=<rgba 1 0.15 0.8867 1>}.RGB)
   B=Power(Base=SphereMask(A=@n8, B=@n9, Radius=@n10, Hardness=@n11) {hardnessPercent=80}, Exp=@n12)
  )
  n14 = Time()
  n15 = Multiply(A=@n14, B=ScalarParameter() {defaultValue=1, parameterName=Flicker_Speed})
  n16 = Sine(None=@n15) {period=0.155}
  n17 = Sine(None=@n15) {period=0.2}
  n18 = Clamp(
   None=Multiply(A=Sine(None=@n15) {period=0.266}, B=Multiply(A=@n16, B=@n17))
   Min=ScalarParameter() {parameterName=Flicker-ClampMin}
   Max=ScalarParameter() {defaultValue=3, parameterName=Flicker-ClampMax}
  ) {maxDefault=3, minDefault=0.5}
  n19 = ScalarParameter() {defaultValue=4, parameterName=Flicker-Brightness}
  n20 = LinearInterpolate(B=Add(A=Multiply(A=@n18, B=@n19)) {constB=0.5}, Alpha=Floor(None=Add(A=@n6) {constB=0.5})) {constA=1}
  n21 = Subtract(A=@n5)
  n22 = SphereMask(A=@n8, B=Constant() {r=1.25}, Radius=@n10, Hardness=@n11) {hardnessPercent=80}
  n23 = VectorParameter() {parameterName=Color-PrimaryGlow, defaultValue=<rgba 5 0 4.3333 1>}
  n24 = Multiply(A=@n20, B=Multiply(A=Power(Base=@n22, Exp=@n12), B=@n23.RGB))
  n25 = TextureCoordinate() {coordinateIndex=1}
  n26 = Multiply(A=@n14, B=ScalarParameter() {defaultValue=0.35, parameterName=Time})
  n27 = TextureSampleParameter2D(UVs=Panner(Coordinate=Multiply(A=@n25) {constB=0.9}, Time=@n26) {speedX=-0.1115, speedY=0.1}) {parameterName=UVNoise-Map, samplerType=SAMPLERTYPE_LinearGrayscale, texture=/Game/Global/Textures/Default/Noise/T_SwirlNoise_N_height}
  n28 = TextureSampleParameter2D(UVs=Panner(Coordinate=Multiply(A=@n25) {constB=-0.85}, Time=@n26) {speedX=-0.125, speedY=-0.163}) {parameterName=UVNoise-Map, samplerType=SAMPLERTYPE_LinearGrayscale, texture=&T_SwirlNoise_N_height}
  n29 = Divide(A=Add(A=@n27.RGB, B=@n28.RGB))
  n30 = Multiply(A=ComponentMask(None=@n29) {g=true, r=true}) {constB=2}
  n31 = Add(A=@n25, B=Multiply(A=Subtract(A=@n30), B=ScalarParameter() {defaultValue=0.09, parameterName=UV-DistortionStrength}))
  n32 = TextureSample(UVs=Clamp(None=Rotator(Coordinate=@n31) {speed=-0.35})) {texture=&/vfx_Saucer_Spiral_01}
  n33 = LinearInterpolate(B=Power(Base=@n29), Alpha=ScalarParameter() {defaultValue=1, parameterName=NoiseInfluenceOnColor}) {constA=1}
  n34 = TextureCoordinate() {coordinateIndex=1}
  n35 = Constant() {r=0.5}
  n36 = Power(
   Base=SphereMask(
    A=ComponentMask(None=@n34) {g=true, r=true}
    B=AppendVector(A=@n35, B=@n35)
    Radius=ScalarParameter() {defaultValue=0.5, parameterName=SphereMask-Radius}
    Hardness=ScalarParameter() {parameterName=SphereMask-Hardness}
   ) {hardnessPercent=80}
   Exp=ScalarParameter() {defaultValue=2.2, parameterName=SphereMask-Exponent}
  )
  n37 = Multiply(A=Multiply(A=Multiply(A=@n32.RGB, B=@n33), B=Clamp(None=@n36)), B=@n23.RGB)
  n38 = Multiply(A=ComponentMask(None=TextureCoordinate()) {g=true}) {constB=2}
  n39 = LinearInterpolate(
   A=Multiply(A=Multiply(A=@n20, B=@n13), B=Clamp(None=@n21)) {desc=Panning Lines}
   B=Multiply(A=Add(A=@n37, B=@n24), B=Clamp(None=@n5)) {desc=Iris and Solid No Panning}
   Alpha=Clamp(None=Floor(None=@n38))
  )
  MaterialFunctionCall(LDR Input=Clamp(None=@n39), HDR Tint=@n39, HDR Intensity Multiplier=ScalarParameter() {defaultValue=1, parameterName=Neon HDR INtensity}) {materialFunction=&DeriveHDRfromLDR}
  ScalarParameter() {group=Glass, parameterName=Metallic, sortPriority=1}
  ScalarParameter() {defaultValue=0.2, group=Glass, parameterName=Roughness, sortPriority=3}
  ScalarParameter() {defaultValue=1, group=Glass, parameterName=Specular, sortPriority=2}
  TextureSample() {samplerType=SAMPLERTYPE_Normal, texture=/Game/Weapons/FORT_Rifles/Textures/Neon_Glow_GlowParts_N}
 ]
}
